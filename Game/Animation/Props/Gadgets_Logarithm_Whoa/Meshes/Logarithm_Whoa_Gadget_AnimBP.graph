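# Animation Blueprint for the Logarithm Whoa gadget, a glasses prop that fits itself to the wearing player's character.
# On initialization it derives the character ID and variant tags, picks an attach data table row (variant, character, then gender default), and attaches itself to that body part's socket.
# The AnimGraph applies the row's offset to the `offset` bone, shrinking it to hide disabled glasses.
# A `logarithm_enable` curve value of 1 re-enables them.
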AnimBlueprint /Game/Animation/Props/Gadgets_Logarithm_Whoa/Meshes/Logarithm_Whoa_Gadget_AnimBP
target_skeleton=&/Logarithm_Whoa_Gadget_Skeleton
graphs [
 AnimGraph {nodes [
   n1 = Input Pose()
   n2 = "Transform (Modify) Bone - Bone: offset"(
    ComponentPose=Local To Component(LocalPose=Slot 'DefaultSlot'(Source=@n1))
    Alpha=1
    Translation=Get Local Offset T() -> (Local Offset T:Vector)
    Rotation=Get Local Offset R() -> (Local Offset R:Rotator)
    Scale=Get Local Offset S() -> (Local Offset S:Vector)
   ) {
    rotationMode=BMM_Replace
    rotationSpace=BCS_ParentBoneSpace
    scaleMode=BMM_Replace
    scaleSpace=BCS_ParentBoneSpace
    translationMode=BMM_Replace
    translationSpace=BCS_ParentBoneSpace
    alphaBoolBlend {blendOption=Linear}
    boneToModify {boneName=offset}
    rotation {roll=-10}
    scale=<1.5 1.5 1.5>
    translation {y=9, z=6.5}
   }
   n3 = "Transform (Modify) Bone - Bone: offset"!(
    ComponentPose=@n2
    bAlphaBoolEnabled:bool=Get Glasses Disabled() -> (Glasses Disabled:bool)
    Translation="0,0,0"
    Rotation="0,0,0"
    Scale="0.01,0.01,0.01"
   ) -> (Pose) {
    alphaInputType=Bool
    rotationSpace=BCS_ParentBoneSpace
    scaleMode=BMM_Replace
    scaleSpace=BCS_ParentBoneSpace
    translationSpace=BCS_ParentBoneSpace
    alphaBoolBlend {blendOption=Linear}
    boneToModify {boneName=offset}
    scale=<0.01 0.01 0.01>
   }
   Output Pose(Result=Component To Local(ComponentPose=@n3)) {layerGroup=None}
  ]}
 EventGraph {nodes [
   PrintString!(execute, InString=Not Player Pawn, bPrintToScreen, bPrintToLog, TextColor, Duration, Key) -> (then)
   > n1 = Cast To SoldierMannequin_Male!(execute, Object:Wildcard) -> (then, CastFailed, AsSoldier Mannequin Male:Soldier Mannequin Male Object Reference) {is=title K2Node_DynamicCast}
   > n2 = Set currentCharacter(currentCharacter:String=GetDisplayName(Object=GetAthenaCharacterItemDefinition!(self:Fort Player Mannequin Object Reference=@n1) -> (ReturnValue:Athena Character Item Definition Object Reference))) -> (Output_Get:String)
   > n3 = Set currentVariant(currentVariant:String) -> (Output_Get:String)
   > PrintString(InString=Append!(A="Mannequin: ", B=@n2, C:String=., D:String=@n3) -> (ReturnValue))
   PrintString(execute=@n1.CastFailed, InString="Unknown Entity running  Capsule_Whoa_Gadget_AnimBP")
   Event BlueprintInitializeAnimation()
   > n4 = Sequence!(execute) -> (then_0, then_1, then_2:Exec, then_3:Exec, then_4:Exec)
   n5 = Sequence!(execute=@n4.then_2) -> (then_0, then_1, then_2)
   n6 = Break Emote Prop Attach Struct!(EmotePropAttachStruct:"Emote Prop Attach Struct Structure (by ref)"=Get Attach Data Row() -> (Attach Data Row:Emote Prop Attach Struct Structure)) -> (
    enabled_18_#1:bool=~
    socket_2_#2:Name
    offsetT_6_#3:Vector=~
    offsetR_9_#4:Rotator=~
    offsetS_15_#5:Vector=~
    part_21_#6:EFortCustomPartType Enum
    variantTag_26_#7:Gameplay Tag Structure
    notes_33_#8:String
   ) {is=title K2Node_BreakStruct}
   > Set Glasses Disabled(Glasses Disabled:bool=NOT Boolean(A=@n6)) -> (Output_Get:bool)
   > n7 = Set Local Offset T(Local Offset T:Vector=@n6.offsetT_6_#3) -> (Output_Get:Vector)
   > n8 = Set Local Offset R(Local Offset R:Rotator=@n6.offsetR_9_#4) -> (Output_Get:Rotator)
   > n9 = Set Local Offset S(Local Offset S:Vector=@n6.offsetS_15_#5) -> (Output_Get:Vector)
   n10 = Get Fort Player Pawn() -> (Fort Player Pawn:Fort Player Pawn Object Reference)
   Is Valid(exec=@n5.then_1, InputObject=@n10)
   n11 = GetSkeletalMeshForPartType(self=@n10, PartType=@n6.part_21_#6)
   > Is Valid(InputObject=@n11)
   > Set Driving Skelmesh(Driving Skelmesh:Skeletal Mesh Component Object Reference=@n11) -> (Output_Get:Skeletal Mesh Component Object Reference)
   > Set Socket(Socket:Name=@n6.socket_2_#2) -> (Output_Get:Name=None)
   n12 = GetOwningActor()
   Is Valid(exec=@n4.then_0, InputObject=@n12)
   > n13 = Cast To FortPlayerPawn(Object=@n12)
   > Set Fort Player Pawn(Fort Player Pawn:Fort Player Pawn Object Reference=@n13) -> (Output_Get:Fort Player Pawn Object Reference)
   n14 = GetCharacterDisplayName!(Actor:Actor Object Reference=@n13) -> (ReturnValue:String)
   > n15 = Branch(Condition=StartsWith(SourceString=@n14, InPrefix=Character_, SearchCase=CaseSensitive))
   n16 = GetItemDefinitionTypeFromLoadoutWithActor!(
    Actor:Actor Object Reference=@n13
    LoadoutSlot:Gameplay Tag Structure=(TagName="LoadoutSlot.Character")
    CosmeticType:Athena Cosmetic Item Definition Class Reference=/Script/FortniteGame.AthenaCharacterItemDefinition
   ) -> (OutCosmeticItemDef:Athena Character Item Definition Object Reference)
   > n17 = Set assetID(assetID:String="To String (Integer)"(InInt=Get BaseShotGridId!(self:Athena Cosmetic Item Definition Object Reference=@n16) -> (BaseShotGridId:int=0))) -> (Output_Get:String)
   n18 = For Each Loop with Break!(
    Exec:Exec=[@n28.then, @n17.then, @n27.then]
    Array:Array of Mcp Variant Channel Info Structures=GetAllLoadoutVariantInfoFromActor!(Actor:Actor Object Reference=@n13) -> (ReturnValue:Array of Mcp Variant Channel Info Structures)
    Break:Exec
   ) -> (LoopBody:Exec, Array Element:Mcp Variant Channel Info Structure, Array Index:int, Completed:Exec)
   n19 = Break Mcp Variant Channel Info!(McpVariantChannelInfo:"Mcp Variant Channel Info Structure (by ref)"=@n18) -> (
    OwnedVariantTags:Gameplay Tag Container Structure
    ItemVariantIsUsedFor:Item Definition Base Object Reference
    CustomData:String
    VariantChannelTag:Gameplay Tag Structure
    ActiveVariantTag:Gameplay Tag Structure
   ) {is=title K2Node_BreakStruct}
   n20 = GetDebugStringFromGameplayTag!(GameplayTag:Gameplay Tag Structure=@n19.ActiveVariantTag) -> (ReturnValue:String)
   n21 = RightChop!(SourceString:String=@n20, Count:int=FindSubstring(SearchIn=@n20, Substring=., bSearchFromEnd=true)) -> (ReturnValue:String)
   > Add(TargetArray=Get currentVariantArray() -> (currentVariantArray:Array of Strings), NewItem=@n21) {is=title K2Node_CallArrayFunction}
   n22 = GetDebugStringFromGameplayTag(GameplayTag=@n19.VariantChannelTag)
   n23 = FindSubstring(SearchIn=@n22, Substring=., bSearchFromEnd=true)
   > Add(TargetArray=Get currentVariantArray(), NewItem=Append!(A=RightChop(SourceString=@n22, Count=@n23), B=@n21) -> (ReturnValue))
   n24 = Split(SourceString=@n14, InStr=_Athena_, SearchDir=FromStart)
   n25 = Split(SourceString=@n24, InStr=CID_, SearchDir=FromStart)
   n26 = Branch(execute=@n15.else, Condition=AND Boolean(A=@n24.ReturnValue, B=@n25.ReturnValue))
   > n27 = Set assetID(assetID=@n25.RightS)
   n28 = Set assetID(execute=@n26.else)
   n29 = GetOwningComponent()
   Is Valid(exec=@n4.then_3, InputObject=@n29)
   > Attach Component To Component!(
    execute
    self:Scene Component Object Reference=@n29
    Parent:Scene Component Object Reference=Get Driving Skelmesh() -> (Driving Skelmesh:Skeletal Mesh Component Object Reference)
    SocketName:Name=Get Socket() -> (Socket:Name=None)
    LocationRule:EAttachmentRule Enum=KeepRelative
    RotationRule:EAttachmentRule Enum=KeepRelative
    ScaleRule:EAttachmentRule Enum=KeepRelative
    bWeldSimulatedBodies:bool=TRUE
   ) -> (then, ReturnValue:bool)
   n30 = For Each Loop with Break(Exec=@n4.then_1, Array:Array of Strings=Get currentVariantArray(), Break=@n49.then) -> (Array Element:String)
   n31 = Branch(execute=@n30.Completed, Condition=Get variantFound() -> (variantFound:bool))
   n32 = GetDataTableRowNames(execute=@n31.else, Table=Get Data Table() -> (Data Table:Data Table Object Reference))
   n33 = Find Item!(TargetArray:Array of Names=@n32, ItemToFind:"Name (by ref)"=String To Name(InString=Get assetID() -> (assetID:String))) -> (ReturnValue)
   > n34 = Branch(Condition="integer > integer"(A=@n33, B=-1) {is=title K2Node_CallFunction})
   > n35 = Get Data Table Row!(
    execute:Exec
    DataTable:Data Table Object Reference=Get Data Table()
    RowName:Name="Get (a ref)"(Array:Array of Names=@n32, Dimension 1=@n33) -> (Output:"Name (by ref)")
   ) -> (then:Exec, RowNotFound:Exec, ReturnValue:Emote Prop Attach Struct Structure)
   > Set Attach Data Row(Attach Data Row:Emote Prop Attach Struct Structure=@n35) -> (Output_Get:Emote Prop Attach Struct Structure)
   n36 = For Each Loop with Break(
    Exec=@n34.else
    Array:Array of Custom Character Part Soft Object References=Get BaseCharacterParts!(self:Athena Character Item Definition Object Reference=@n16) -> (BaseCharacterParts:Array of Custom Character Part Soft Object References)
    Break=@n43.then
   ) -> (Array Element:Custom Character Part Soft Object Reference)
   n37 = Switch on EFortCustomGender!(
    execute=@n36.Completed
    Selection:EFortCustomGender Enum=Get GenderPermitted!(self:Custom Character Part Object Reference=Get bodyPart() -> (bodyPart:Custom Character Part Object Reference)) -> (GenderPermitted:EFortCustomGender Enum=Invalid)
   ) -> (Invalid:Exec, Male:Exec, Female:Exec, Both:Exec) {is=title K2Node_SwitchEnum}
   n38 = Set defaultRowID(execute=[@n37.Both, @n37.Invalid, @n37.Male], defaultRowID:Name=005) -> (Output_Get:Name=None)
   n39 = Get Data Table Row(execute=[@n40.then, @n38.then], DataTable=Get Data Table(), RowName=Get defaultRowID() -> (defaultRowID:Name=None))
   > Set Attach Data Row(Attach Data Row=@n39)
   n40 = Set defaultRowID(execute=@n37.Female, defaultRowID=001) -> (Output_Get=None)
   n41 = Resolve Soft Reference!(Input:Custom Character Part Soft Object Reference=@n36) -> (Output:Custom Character Part Object Reference)
   n42 = GetCharacterPartType!(self:Custom Character Part Object Reference=@n41) -> (ReturnValue:EFortCustomPartType Enum=Head)
   Branch(
    execute=@n36.LoopBody
    Condition="Equal Exactly (String)"(A=Enum to String!(Enumerator:EFortCustomPartType Enum=@n42) -> (ReturnValue:String), B=Body)
   )
   > n43 = Set bodyPart(bodyPart:Custom Character Part Object Reference=@n41) -> (Output_Get:Custom Character Part Object Reference)
   n44 = Branch(execute=@n30.LoopBody, Condition=false)
   > n45 = PrintString(InString=@n30, Duration=5)
   n46 = GetDataTableRowNames(execute=[@n45.then, @n44.else], Table=Get Data Table())
   n47 = String To Name(InString=Append(A=Get assetID(), B=@n30))
   > Branch(Condition="integer > integer"(A=Find Item(TargetArray=@n46, ItemToFind=@n47), B=-1))
   > n48 = Get Data Table Row(DataTable=Get Data Table(), RowName=@n47)
   > Set Attach Data Row(Attach Data Row=@n48)
   > n49 = Set variantFound(variantFound:bool=true) -> (Output_Get:bool)
   Event BlueprintUpdateAnimation()
   > Branch(Condition="Equal (Float)"(A=GetCurveValue(CurveName=logarithm_enable), B=1))
   > Set Glasses Disabled(Glasses Disabled=false)
   MakeTransform(Location=@n7, Rotation=@n8, Scale=@n9)
   TryGetPawnOwner()
  ]}
]
variables [
 assetID {type=string, default=""}
 Attach Data Row {
  type=EmotePropAttachStruct
  default {enabled=true, notes="", part=Head, socket=head, offsetR=<rot 0 0 0>, offsetS=<1 1 1>, offsetT=<0 0 0>, variantTag {tagName=None}}
 }
 bodyPart {type=CustomCharacterPart, default=null}
 currentCharacter {type=string, default=""}
 currentVariant {type=string, default=""}
 currentVariantArray {type="array<string>", default []}
 Data Table {type=DataTable, default=&/Logarithm_Whoa_Gadget_Attach_DataTable}
 defaultRowID {type=string, default=None}
 Driving Skelmesh {type=SkeletalMeshComponent}
 Fort Player Pawn {type=FortPlayerPawn}
 Glasses Disabled {type=boolean}
 Local Offset R {type=Rotator}
 Local Offset S {type=Vector}
 Local Offset T {type=Vector}
 Socket {type=string, default=None}
 variantFound {type=boolean}
]
